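# Translucent, unlit, two-sided Niagara mesh particle material for the RadioPaca pickaxe, drawing a glowing, animated equalizer-style line.
# Twelve quantized UV columns offset layered sine waves, which modulate a rotated, tiled ramp texture and blend two colors.
# The Audio Reactivy dynamic parameter scales animation speed, adjusts vertical falloff and blends out a built-in sine beat dimming.
# UV edge masks, a soft base glow, particle alpha and depth fade shape opacity.
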Material /BRCosmetics/Weapons/FORT_Melee/Pickaxe_RadioPaca/FX/M_Pickaxe_RadioPaca_Line
graph {
 properties {
  bAutomaticallySetUsageInEditor=false
  bUsedWithNiagaraMeshParticles=true
  blendMode=BLEND_Translucent
  shadingModel=MSM_Unlit
  twoSided=true
  usedShadingModels=Unlit
 }
 nodes [
  n1 = ParticleColor()
  n2 = ComponentMask(None=TextureCoordinate()) {g=true}
  n3 = ScalarParameter() {defaultValue=50, parameterName=Top_Linear_Int}
  n4 = Saturate(None=Power(Base=Multiply(A=@n2, B=@n3) {constB=50}))
  n5 = ComponentMask(None=TextureCoordinate()) {r=true}
  n6 = Multiply(A=Multiply(A=@n5, B=OneMinus(None=@n5))) {constB=50}
  n7 = TextureCoordinate()
  n8 = Constant2Vector() {g=0.22}
  n9 = OneMinus(None=ComponentMask(None=Add(A=@n7, B=@n8)) {g=true})
  n10 = Saturate(None=Power(Base=Multiply(A=@n9) {constB=2}) {constExponent=3})
  n11 = Sine(None=Time()) {period=0.4839}
  n12 = If(A=@n11, "A > B"=@n11, "A == B"=@n11, "A < B"=Constant())
  n13 = ScalarParameter() {defaultValue=0.4, parameterName=Beat Intensity, sliderMax=1}
  n14 = Multiply(A=Time(), B=DynamicParameter() {paramNames [Audio Reactivy, Param2, Param3, Param4]}.Audio Reactivy)
  n15 = ComponentMask(None=TextureCoordinate()) {r=true}
  n16 = Constant() {r=12}
  n17 = Divide(A=Floor(None=Multiply(A=@n15, B=@n16)), B=@n16)
  n18 = Frac(None=Add(A=@n14, B=@n17))
  n19 = Sine(None=@n18) {period=0.8}
  n20 = Multiply(A=ConstantBiasScale(None=Sine(None=@n18) {period=0.3})) {constB=0.25}
  n21 = Subtract(A=@n14, B=@n17)
  n22 = ConstantBiasScale(None=Sine(None=Frac(None=@n21)) {period=0.5})
  n23 = Divide(A=Add(A=ConstantBiasScale(None=@n19), B=Add(A=@n20, B=@n22))) {constB=4}
  n24 = MaterialFunctionCall(UVs=TextureCoordinate() {uTiling=12}, "Rotation Angle (0-1)"=Constant() {r=0.25}) {materialFunction=&CustomRotator}
  n25 = ComponentMask(None=TextureCoordinate()) {g=true}
  n26 = DynamicParameter() {paramNames [Audio Reactivy, Param2, Param3, Param4]}
  n27 = Power(Base=Multiply(A=OneMinus(None=@n25), B=LinearInterpolate(Alpha=@n26.Audio Reactivy) {constA=1.5, constB=1.3}) {constB=1.5}) {constExponent=6}
  n28 = Multiply(
   A=@n23
   B=Multiply(
    A=TextureSample(UVs=@n24.Rotated Values) {texture=/Game/Effects/Fort_Effects/Textures/Ramps/T_Power_Source_Sphere_Ramp}.R
    B=Saturate(None=@n27)
   )
  )
  n29 = Multiply(A=OneMinus(None=Multiply(A=@n12, B=@n13) {constB=0.25}), B=@n28)
  n30 = DynamicParameter() {paramNames [Audio Reactivy, Param2, Param3, Param4]}
  n31 = Multiply(
   A=Saturate(None=Power(Base=@n6))
   B=Add(A=Multiply(A=@n10) {constB=0.2}, B=LinearInterpolate(A=@n29, B=@n28, Alpha=@n30.Audio Reactivy) {constAlpha=0})
  )
  DepthFade(Opacity=Multiply(A=@n1.A, B=Multiply(A=@n4, B=@n31)), FadeDistance=ScalarParameter() {parameterName=DepthFade})
  n32 = VectorParameter() {parameterName=Color01, defaultValue=<rgba 0 0.1167 1 1>}
  n33 = VectorParameter() {parameterName=Color02, defaultValue=<rgba 0.134 1.3965 2 1>}
  MaterialFunctionCall(Emissive=Multiply(A=LinearInterpolate(A=@n32.RGB, B=@n33.RGB, Alpha=@n23), B=@n1.RGB)) {materialFunction=&MF_FX_EmissiveAdjustments}
 ]
}
