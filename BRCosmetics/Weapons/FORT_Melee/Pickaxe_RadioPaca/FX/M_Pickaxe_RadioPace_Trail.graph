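# Unlit, translucent, two-sided Niagara ribbon trail material for the RadioPaca pickaxe, drawing a blocky, equalizer-like animated bar pattern.
# Quantized UVs form stepped bars whose lengths come from summed time-driven sine waves, modulated by a periodic pulse.
# A pixelated, scrolling noise texture breaks up the bar mask, which feeds particle alpha and depth fade.
# Noise also shades the emissive between dark and bright tones, with a VariantSwitch dynamic parameter blending blue and green palettes.
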Material /BRCosmetics/Weapons/FORT_Melee/Pickaxe_RadioPaca/FX/M_Pickaxe_RadioPace_Trail
graph {
 properties {
  bAutomaticallySetUsageInEditor=false
  bUsedWithNiagaraRibbons=true
  blendMode=BLEND_Translucent
  shadingModel=MSM_Unlit
  twoSided=true
  usedShadingModels=Unlit
 }
 nodes [
  n1 = OneMinus(None=ComponentMask(None=TextureCoordinate()) {r=true})
  n2 = Ceil(None=Constant() {r=8})
  n3 = Divide(A=Floor(None=Multiply(A=@n1, B=@n2) {constB=10}), B=@n2)
  n4 = Multiply(A=Time(), B=Constant() {r=1})
  n5 = ComponentMask(None=TextureCoordinate()) {g=true}
  n6 = Ceil(None=Constant() {r=8})
  n7 = Divide(A=Floor(None=Multiply(A=@n5, B=@n6)), B=@n6)
  n8 = Frac(None=Add(A=@n4, B=@n7))
  n9 = ConstantBiasScale(None=Sine(None=@n8) {period=0.3})
  n10 = Sine(None=Frac(None=Subtract(A=@n4, B=@n7))) {period=0.5}
  n11 = Add(A=ConstantBiasScale(None=Sine(None=@n8) {period=0.8}), B=Add(A=Multiply(A=@n9) {constB=0.25}, B=ConstantBiasScale(None=@n10)))
  n12 = Sine(None=Time()) {period=0.4839}
  n13 = If(A=@n12, "A > B"=@n12, "A == B"=@n12, "A < B"=Constant())
  n14 = Constant() {r=0.35}
  n15 = Multiply(A=Add(A=Divide(A=@n11) {constB=4}) {constB=0.3}, B=OneMinus(None=Multiply(A=@n13, B=@n14) {constB=0.25}))
  n16 = If(A=Add(A=@n3) {constB=0.1}, B=@n15, "A > B"=Constant() {r=1}, "A < B"=Constant())
  Add(A=@n16) {constB=0}
  n17 = ParticleColor()
  n18 = Constant() {r=8}
  n19 = Floor(None=Multiply(A=TextureCoordinate() {uTiling=3}, B=@n18))
  n20 = Time()
  n21 = Constant2Vector() {r=-0.2}
  n22 = TextureSample(UVs=Add(A=Divide(A=@n19, B=@n18), B=Multiply(A=@n20, B=@n21))) {
   samplerSource=SSM_Wrap_WorldGroupSettings
   samplerType=SAMPLERTYPE_LinearColor
   texture=/Game/Characters/Player/Male/Medium/Bodies/M_MED_EmeraldGlassGreen/Textures/T_CowlNoise
  }
  n23 = Multiply(A=@n22.RGB, B=Saturate(None=@n16))
  n24 = ComponentMask(None=TextureCoordinate()) {r=true}
  n25 = Floor(None=Multiply(A=OneMinus(None=@n24)) {constB=1.09})
  n26 = Add(A=@n3) {constB=-0.2}
  n27 = Constant() {r=1}
  n28 = Constant()
  n29 = Saturate(None=Add(A=@n25, B=If(A=@n26, B=@n15, "A > B"=@n27, "A < B"=@n28)))
  DepthFade(Opacity=Multiply(A=@n17.A, B=Add(A=@n23, B=@n29)), FadeDistance=ScalarParameter() {defaultValue=1, parameterName=DepthFade})
  n30 = LinearInterpolate(Alpha=@n22.RGB) {constA=0.04}
  n31 = LinearInterpolate(A=Constant3Vector() {constant=<rgba 0.02 0.2384 1 1>}, B=Constant3Vector() {constant=<rgba 0.5938 1.7963 3 1>}, Alpha=@n30)
  n32 = LinearInterpolate(A=Constant3Vector() {constant=<rgba 0.1472 0.25 0.0375 1>}, B=Constant3Vector() {constant=<rgba 0.6201 1 0.1562 1>}, Alpha=@n30)
  n33 = DynamicParameter() {defaultValue=<rgba 0 1 1 1>, paramNames [VariantSwitch, Param2, Param3, Param4]}
  MaterialFunctionCall(Emissive=Multiply(A=LinearInterpolate(A=@n31, B=@n32, Alpha=@n33.VariantSwitch), B=@n17.RGB)) {materialFunction=&MF_FX_EmissiveAdjustments}
  Multiply()
  Multiply(A=@n22.R, B=@n25)
 ]
}
