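# Niagara idle effect for the RadioPaca pickaxe: a disc sprite, a center glow and pulsing mesh outlines, tinted blue to green by VariantSwitch.
# A system-level InterpolateOverTime moving average drives glow brightness and, via a threshold, selects between OutLine_FEDefault's single pulse and OutLine_Reactivity's staggered double pulse.
# User "Is Reactive?" scales glow and outlines from zero.
# bIsFrontEndPreview raises minimum glow brightness and gates the default outline's size.
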NiagaraSystem /BRCosmetics/Weapons/FORT_Melee/Pickaxe_RadioPaca/FX/NS_Pickaxe_RadioPaca_Idle02
dependencies {
 usedDataInterfaces [&NiagaraDataInterfaceCurve, &NiagaraDataInterfaceColorCurve]
 usedDynamicInputs []
 usedModules [
  &SystemState
  &InterpolateOverTime
  &self:SystemSpawnScript.SystemScriptSource.SystemScriptGraph.NiagaraNodeAssignment_0.SetVariables_#1
 ]
}
emitters {
 CenterGlow {
  emitterUpdateScript [
   &EmitterState(
    Inactive Response=::NewEnumerator1
    Life Cycle Mode=::NewEnumerator1
    Loop Behavior=::NewEnumerator1
    Loop Duration Mode=::NewEnumerator1
    UseLoopDelay=false
   )
   &SpawnBurst_Instantaneous(Use Spawn Probability=false, UseLoopCountLimit=false)
  ]
  particleSpawnScript [
   &InitializeParticle(
    Color=&LerpLinearColors(StartColor=<rgba 0.0347 0.1736 0.6667 1>, EndColor=<rgba 0.2036 0.3646 0.0217 1>, LerpFactor=$User.VariantSwitch)
    Lifetime=5
    Position Offset=<1 0 0>
    Sprite Size Mode=::NewEnumerator3
    Uniform Sprite Size=40
    UsePositionOffset=true
    Write Material Random=true
   )
   &SpriteFacingAndAlignment(WriteAlignment=false, WriteFacing=true)
  ]
  particleUpdateScript [
   &ParticleState(Kill Particles When Lifetime Has Elapsed=false, Let Infinitely Lived Particles Die When Emitter Deactivates=true)
   &ScaleSpriteSize(
    Scale Sprite Size Mode=::NewEnumerator0
    Uniform Scale Factor=&Multiply_Float(A=1, B=&Lerp_Float(A=0, B=1, Alpha=Linked() {linkedVariable {name=User.Is Reactive?, type=NiagaraFloat}}))
   )
   &ScaleColor(
    Scale RGB=&Multiply_VectorByFloat(
     Vector=<1 1 1>
     Float=&MakeCustomFloatFromBool(
      Bool=$User.bIsFrontEndPreview
      True Float=&Lerp_Float(A=1.5, B=2, Alpha=$System.InterpolateOverTime.InterpolateOverTime_Float.Moving Average)
      False Float=&Lerp_Float(A=0, B=2, Alpha=$System.InterpolateOverTime.InterpolateOverTime_Float.Moving Average)
     )
    )
    ScaleA=true
    ScaleRGB=true
   )
   /SetVariables(Particles.Color002=&Multiply_LinearColorByFloat(
     Linear Color=&LerpLinearColors(StartColor=<rgba 0.115 0.8444 1 1>, EndColor=<rgba 0.6649 1 0.0808 1>, LerpFactor=$User.VariantSwitch)
     Float=&Lerp_Float(A=0, B=3, Alpha=$System.InterpolateOverTime.InterpolateOverTime_Float.Moving Average)
    )) {isSetParameters=true}
   /SetVariables(Particles.Size002:Vector2f=&Vector2DFromFloat(Value=&Multiply_Float(A=20, B=&Lerp_Float(A=0, B=1, Alpha=Linked() {linkedVariable {name=User.Is Reactive?, type=NiagaraFloat}})))) {isSetParameters=true}
   &DynamicMaterialParameters(
    Index 0 Param 1=2
    Index 0 Param 2=0.5
    Module.Index0Float0Write=true
    Module.Index0Float1Write=true
    Module.Index0Float2Write=true
    Module.Index0Float3Write=true
   )
  ]
  rendererClasses [&NiagaraSpriteRendererProperties]
  renderers [
   {rendererClass=&NiagaraSpriteRendererProperties}
   {rendererClass=&NiagaraSpriteRendererProperties, rendererIndex=1}
  ]
 }
 Disc {
  emitterUpdateScript [
   &EmitterState(
    Inactive Response=::NewEnumerator1
    Life Cycle Mode=::NewEnumerator1
    Loop Behavior=::NewEnumerator1
    Loop Duration Mode=::NewEnumerator1
    UseLoopDelay=false
   )
   &SpawnBurst_Instantaneous(Use Spawn Probability=false, UseLoopCountLimit=false)
  ]
  particleSpawnScript [
   &InitializeParticle(
    Color Mode=::NewEnumerator0
    Lifetime=5
    Position Offset=<0.1 0 0>
    Sprite Rotation Angle=10
    Sprite Rotation Mode=::NewEnumerator0
    Sprite Size Mode=::NewEnumerator3
    Uniform Sprite Size=43
    UsePositionOffset=true
    Write Material Random=true
   )
   &SpriteFacingAndAlignment(WriteAlignment=false, WriteFacing=true)
  ]
  particleUpdateScript [
   &ParticleState(Kill Particles When Lifetime Has Elapsed=false, Let Infinitely Lived Particles Die When Emitter Deactivates=true)
   &SpriteRotationRate(Rotation Rate=&FloatFromCurve(
     FloatCurve=DataInterface() {propertyValues {Curve {keys [
         {interpMode=RCIM_Cubic, tangentMode=RCTM_User, time=0, value=0.25}
         {interpMode=RCIM_Cubic, tangentMode=RCTM_User, time=0.2, value=0.5}
         {arriveTangent=-0.0568, interpMode=RCIM_Cubic, leaveTangent=-0.0568, tangentMode=RCTM_User, time=0.5, value=0}
         {arriveTangent=0.0228, interpMode=RCIM_Cubic, leaveTangent=0.0228, tangentMode=RCTM_User, time=0.8, value=0.375}
         <at 1 0.25 RCIM_Cubic>
        ]}}}
     CurveIndex=&Frac_Float(Float=$Particles.NormalizedAge)
     Scale Curve=-5
    )) {enabled=false}
  ]
  rendererClasses [&NiagaraSpriteRendererProperties]
  renderers [{rendererClass=&NiagaraSpriteRendererProperties}]
 }
 OutLine_FEDefault {
  emitterUpdateScript [
   &EmitterState(Life Cycle Mode=::NewEnumerator1, Loop Duration=1, UseLoopDelay=false)
   &SpawnBurst_Instantaneous(
    Spawn Count=&MakeIntFromFloat(Float=&MakeCustomFloatFromBool(
      Bool=&SetBoolByFloatComparison(A=$System.InterpolateOverTime.InterpolateOverTime_Float.Moving Average, B=0.1)
      True Float=0
      False Float=1
     ))
    Use Spawn Probability=false
    UseLoopCountLimit=false
   )
  ]
  particleSpawnScript [
   &InitializeParticle(
    Color Mode=::NewEnumerator0
    Lifetime=0.8
    Mesh Scale Mode=::NewEnumerator3
    Mesh Uniform Scale=0.19
    UsePositionOffset=true
    Write Material Random=true
   )
   &InitialMeshOrientation(Mesh Orientation Mode=::NewEnumerator2, Rotation=<0 0.25 0>, Use Rotation Vector=true)
   &InitialMeshOrientation(Mesh Orientation Mode=::NewEnumerator2, Use Rotation Vector=true) {module=InitialMeshOrientation001}
  ]
  particleUpdateScript [
   &ParticleState()
   &ScaleMeshSize(Scale Factor=&VectorFromFloat(Value=&FloatFromCurve(
      FloatCurve=DataInterface() {propertyValues {Curve {keys [<at 0 1 RCIM_Cubic>, {arriveTangent=0.1444, interpMode=RCIM_Cubic, leaveTangent=0.1444, tangentMode=RCTM_User, time=1, value=1.09}]}}}
      Scale Curve=&MakeCustomFloatFromBool(Bool=$User.bIsFrontEndPreview, True Float=1.03, False Float=0)
     )))
   &ScaleMeshSize(Scale Factor=&VectorFromFloat(Value=&Lerp_Float(A=0, B=1, Alpha=Linked() {linkedVariable {name=User.Is Reactive?, type=NiagaraFloat}}))) {module=ScaleMeshSize001}
   &DynamicMaterialParameters(
    Index 0 Param 1=&FloatFromCurve(FloatCurve=DataInterface() {propertyValues {Curve {keys [<at 0 0.3>,<at 1 0.1>]}}})
    Module.Index0Float0Write=true
    Module.Index0Float1Write=true
    Module.Index0Float2Write=true
    Module.Index0Float3Write=true
   )
   &ScaleColor(
    Scale Alpha=&FloatFromCurve(FloatCurve=DataInterface() {propertyValues {Curve {
        defaultValue=1
        keys [
         <at 0 0 RCIM_Cubic>
         {arriveTangent=-0.0026, interpMode=RCIM_Cubic, leaveTangent=-0.0026, tangentMode=RCTM_User, time=0.3, value=1}
         {interpMode=RCIM_Cubic, tangentMode=RCTM_User, time=0.7, value=1}
         {interpMode=RCIM_Cubic, tangentMode=RCTM_User, time=1, value=0}
        ]
       }}})
    Scale RGB=&MakeVectorFromLinearColorRGB(Color=&LerpLinearColors(
      StartColor=&ColorFromCurve(ColorCurve=DataInterface() {propertyValues {
         AlphaCurve {keys []}
         BlueCurve {keys [<at 0.2566 1 RCIM_Cubic>,<at 0.7515 0.849 RCIM_Cubic>]}
         GreenCurve {keys [<at 0.2566 0.665 RCIM_Cubic>,<at 0.7515 0.197 RCIM_Cubic>]}
         RedCurve {keys [<at 0.2566 0.0453 RCIM_Cubic>,<at 0.7515 0.0091 RCIM_Cubic>]}
        }})
      EndColor=&ColorFromCurve(ColorCurve=DataInterface() {propertyValues {
         AlphaCurve {keys []}
         BlueCurve {keys [<at 0.2566 0.1104 RCIM_Cubic>,<at 0.7515 0.0265 RCIM_Cubic>]}
         GreenCurve {keys [<at 0.2566 1 RCIM_Cubic>,<at 0.7515 0.3177 RCIM_Cubic>]}
         RedCurve {keys [<at 0.2566 0.6798 RCIM_Cubic>,<at 0.7515 0.2056 RCIM_Cubic>]}
        }})
      LerpFactor=$User.VariantSwitch
     ))
    ScaleA=true
    ScaleRGB=true
   )
   &ScaleColor(
    Scale Alpha=&FloatFromCurve(FloatCurve=DataInterface() {propertyValues {Curve {
        defaultValue=1
        keys [
         <at 0 0 RCIM_Cubic>
         {arriveTangent=-0.0026, interpMode=RCIM_Cubic, leaveTangent=-0.0026, tangentMode=RCTM_User, time=0.2, value=1}
         {interpMode=RCIM_Cubic, tangentMode=RCTM_User, time=0.7, value=1}
         {interpMode=RCIM_Cubic, tangentMode=RCTM_User, time=1, value=0}
        ]
       }}})
    Scale RGB=&MakeVectorFromLinearColorRGB(Color=&ColorFromCurve(ColorCurve=DataInterface() {propertyValues {
        AlphaCurve {keys []}
        BlueCurve {keys [<at 0.2566 1 RCIM_Cubic>,<at 0.7515 0.849 RCIM_Cubic>]}
        GreenCurve {keys [<at 0.2566 0.665 RCIM_Cubic>,<at 0.7515 0.197 RCIM_Cubic>]}
        RedCurve {keys [<at 0.2566 0.0453 RCIM_Cubic>,<at 0.7515 0.0091 RCIM_Cubic>]}
       }}))
    ScaleA=true
    ScaleRGB=true
   ) {enabled=false, module=ScaleColor001}
  ]
  rendererClasses [&NiagaraMeshRendererProperties]
  renderers [{rendererClass=&NiagaraMeshRendererProperties, rendererIndex=0}]
 }
 OutLine_Reactivity {
  emitterUpdateScript [
   &EmitterState(Life Cycle Mode=::NewEnumerator1, Loop Duration=1, UseLoopDelay=false)
   &SpawnBurst_Instantaneous(
    Spawn Count=&MakeIntFromFloat(Float=&MakeCustomFloatFromBool(
      Bool=&SetBoolByFloatComparison(A=$System.InterpolateOverTime.InterpolateOverTime_Float.Moving Average, B=0.1)
      True Float=1
      False Float=0
     ))
    Use Spawn Probability=false
    UseLoopCountLimit=false
   )
   &SpawnBurst_Instantaneous(
    Spawn Count=&MakeIntFromFloat(Float=&MakeCustomFloatFromBool(
      Bool=&SetBoolByFloatComparison(A=$System.InterpolateOverTime.InterpolateOverTime_Float.Moving Average, B=0.1)
      True Float=1
      False Float=0
     ))
    Spawn Time=0.15
    Use Spawn Probability=false
    UseLoopCountLimit=false
   ) {module=SpawnBurst_Instantaneous002}
  ]
  particleSpawnScript [
   &InitializeParticle(
    Color Mode=::NewEnumerator0
    Lifetime=0.8
    Mesh Scale Mode=::NewEnumerator3
    Mesh Uniform Scale=0.19
    UsePositionOffset=true
    Write Material Random=true
   )
   &InitialMeshOrientation(Mesh Orientation Mode=::NewEnumerator2, Rotation=<0 0.25 0>, Use Rotation Vector=true)
   &InitialMeshOrientation(Mesh Orientation Mode=::NewEnumerator2, Use Rotation Vector=true) {module=InitialMeshOrientation001}
  ]
  particleUpdateScript [
   &ParticleState()
   &ScaleMeshSize(Scale Factor=&VectorFromFloat(Value=&FloatFromCurve(
      FloatCurve=DataInterface() {propertyValues {Curve {keys [<at 0 1 RCIM_Cubic>, {arriveTangent=0.597, interpMode=RCIM_Cubic, leaveTangent=0.597, tangentMode=RCTM_User, time=1, value=1.2625}]}}}
      Scale Curve=1.03
     )))
   &ScaleMeshSize(Scale Factor=&VectorFromFloat(Value=&Lerp_Float(A=0, B=1, Alpha=Linked() {linkedVariable {name=User.Is Reactive?, type=NiagaraFloat}}))) {module=ScaleMeshSize001}
   &DynamicMaterialParameters(
    Index 0 Param 1=&FloatFromCurve(FloatCurve=DataInterface() {propertyValues {Curve {keys [<at 0 0.3>,<at 1 0.1>]}}}, Scale Curve=1.1)
    Module.Index0Float0Write=true
    Module.Index0Float1Write=true
    Module.Index0Float2Write=true
    Module.Index0Float3Write=true
   )
   &ScaleColor(
    Scale Alpha=&FloatFromCurve(FloatCurve=DataInterface() {propertyValues {Curve {
        defaultValue=1
        keys [
         <at 0 0 RCIM_Cubic>
         {arriveTangent=-0.0026, interpMode=RCIM_Cubic, leaveTangent=-0.0026, tangentMode=RCTM_User, time=0.3, value=1}
         {interpMode=RCIM_Cubic, tangentMode=RCTM_User, time=0.7, value=1}
         {interpMode=RCIM_Cubic, tangentMode=RCTM_User, time=1, value=0}
        ]
       }}})
    Scale RGB=&MakeVectorFromLinearColorRGB(Color=&LerpLinearColors(
      StartColor=&ColorFromCurve(ColorCurve=DataInterface() {propertyValues {
         AlphaCurve {keys []}
         BlueCurve {keys [<at 0.2566 1 RCIM_Cubic>,<at 0.7515 0.849 RCIM_Cubic>]}
         GreenCurve {keys [<at 0.2566 0.665 RCIM_Cubic>,<at 0.7515 0.197 RCIM_Cubic>]}
         RedCurve {keys [<at 0.2566 0.0453 RCIM_Cubic>,<at 0.7515 0.0091 RCIM_Cubic>]}
        }})
      EndColor=&ColorFromCurve(ColorCurve=DataInterface() {propertyValues {
         AlphaCurve {keys []}
         BlueCurve {keys [<at 0.2566 0.1104 RCIM_Cubic>,<at 0.7515 0.0265 RCIM_Cubic>]}
         GreenCurve {keys [<at 0.2566 1 RCIM_Cubic>,<at 0.7515 0.3177 RCIM_Cubic>]}
         RedCurve {keys [<at 0.2566 0.6798 RCIM_Cubic>,<at 0.7515 0.2056 RCIM_Cubic>]}
        }})
      LerpFactor=$User.VariantSwitch
     ))
    ScaleA=true
    ScaleRGB=true
   )
   &ScaleColor(
    Scale Alpha=&FloatFromCurve(FloatCurve=DataInterface() {propertyValues {Curve {
        defaultValue=1
        keys [
         <at 0 0 RCIM_Cubic>
         {arriveTangent=-0.0026, interpMode=RCIM_Cubic, leaveTangent=-0.0026, tangentMode=RCTM_User, time=0.2, value=1}
         {interpMode=RCIM_Cubic, tangentMode=RCTM_User, time=0.7, value=1}
         {interpMode=RCIM_Cubic, tangentMode=RCTM_User, time=1, value=0}
        ]
       }}})
    Scale RGB=&MakeVectorFromLinearColorRGB(Color=&ColorFromCurve(ColorCurve=DataInterface() {propertyValues {
        AlphaCurve {keys []}
        BlueCurve {keys [<at 0.2566 1 RCIM_Cubic>,<at 0.7515 0.849 RCIM_Cubic>]}
        GreenCurve {keys [<at 0.2566 0.665 RCIM_Cubic>,<at 0.7515 0.197 RCIM_Cubic>]}
        RedCurve {keys [<at 0.2566 0.0453 RCIM_Cubic>,<at 0.7515 0.0091 RCIM_Cubic>]}
       }}))
    ScaleA=true
    ScaleRGB=true
   ) {enabled=false, module=ScaleColor001}
  ]
  rendererClasses [&NiagaraMeshRendererProperties]
  renderers [{rendererClass=&NiagaraMeshRendererProperties, rendererIndex=0}]
 }
}
summary {
 systemName=NS_Pickaxe_RadioPaca_Idle02
 emitters [
  {emitterName=Disc, rendererClasses [&NiagaraSpriteRendererProperties]}
  {emitterName=CenterGlow, rendererClasses [&NiagaraSpriteRendererProperties]}
  {emitterName=OutLine_FEDefault, rendererClasses [&NiagaraMeshRendererProperties]}
  {emitterName=OutLine_Reactivity, rendererClasses [&NiagaraMeshRendererProperties]}
 ]
 userVariables [
  User.bIsFrontEnd {type=NiagaraBool}
  User.bIsFrontEndPreview {type=NiagaraBool}
  User.Is Reactive? {type=NiagaraFloat, defaultValue=0, description=""}
  User.VariantSwitch {type=NiagaraFloat, defaultValue=0, description=""}
 ]
}
