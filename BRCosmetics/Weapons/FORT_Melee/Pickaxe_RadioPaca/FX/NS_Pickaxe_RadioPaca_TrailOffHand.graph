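# Niagara off-hand trail effect for the RadioPaca pickaxe.
# Three ribbon emitters (MainTrail, Outline, Centerglow) build short-lived anim trails between socket pairs, fading with emitter age and scaled by time of day.
# A brief Particle burst adds rising curl-noise sparks, drawn by two sprite renderers with derived glow size and color attributes.
# User.VariantSwitch blends blue to yellow-green in Outline, Centerglow and sparks, and drives a MainTrail dynamic material parameter.
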NiagaraSystem /BRCosmetics/Weapons/FORT_Melee/Pickaxe_RadioPaca/FX/NS_Pickaxe_RadioPaca_TrailOffHand
dependencies {
 usedDataInterfaces [&NiagaraDataInterfaceCurve, &NiagaraDataInterfaceColorCurve, &NiagaraDataInterfaceSocketReader]
 usedDynamicInputs []
 usedModules [
  &self:SystemSpawnScript.SystemScriptSource.SystemScriptGraph.NiagaraNodeAssignment_0.SetVariables_#1
  &SystemState
 ]
}
emitters {
 Centerglow {
  emitterUpdateScript [
   &EmitterState(Life Cycle Mode=::NewEnumerator1, Loop Behavior=::NewEnumerator1, Loop Duration=0.5, UseLoopDelay=false)
   &FN_SpawnAnimTrail()
  ]
  particleSpawnScript [
   &FN_InitializeAnimTrailBoneIndecies()
   &FN_InitializeAnimTrail(AnimTrailWidth=1.1, Color=<rgba 0.5312 0.5312 0.5312 1>, Lifetime=0.16)
  ]
  particleUpdateScript [
   &FN_UpdateAnimTrail()
   &ScaleColor(
    Scale Alpha=&FloatFromCurve(
     FloatCurve=DataInterface() {propertyValues {Curve {keys [<at 0 1 RCIM_Cubic>,<at 1 0 RCIM_Cubic>]}}}
     Scale Curve=&Multiply_Float(
      A=&FloatFromCurve(FloatCurve=DataInterface() {propertyValues {Curve {keys [<at 0 1>,<at 0.2 1>,<at 0.3 0>]}}}, CurveIndex=$Emitter.Age)
      B=&FNTimeOfDayScalar(Night Scale=0.8, Morning Scale=0.8, Evening Scale=0.8)
     )
    )
    Scale RGB=&MakeVectorFromLinearColorRGB(Color=&LerpLinearColors(
      StartColor=&ColorFromCurve(ColorCurve=DataInterface() {propertyValues {
         AlphaCurve {keys []}
         BlueCurve {keys [{arriveTangent=-17.2603, interpMode=RCIM_Cubic, leaveTangent=-17.2603, tangentMode=RCTM_User, time=0.06, value=8}, <at 0.6278 1 RCIM_Cubic>]}
         GreenCurve {keys [
           {arriveTangent=-14.6689, interpMode=RCIM_Cubic, leaveTangent=-14.6689, tangentMode=RCTM_User, time=0.06, value=5.8271}
           <at 0.6278 0.1416 RCIM_Cubic>
          ]}
         RedCurve {keys [<at 0.06 0 RCIM_Cubic>,<at 0.6278 0 RCIM_Cubic>]}
        }})
      EndColor=&ColorFromCurve(ColorCurve=DataInterface() {propertyValues {
         AlphaCurve {keys []}
         BlueCurve {keys [{arriveTangent=-17.2603, time=0.06, value=0.2499}, <at 0.5594 0.0143>]}
         GreenCurve {keys [{arriveTangent=-14.6689, time=0.06, value=4}, <at 0.5594 0.1823>]}
         RedCurve {keys [<at 0.06 2.1536>,<at 0.5594 0.1108>]}
        }})
      LerpFactor=$User.VariantSwitch
     ))
    ScaleA=true
    ScaleRGB=true
   )
  ]
  rendererClasses [&NiagaraRibbonRendererProperties]
  renderers [{rendererClass=&NiagaraRibbonRendererProperties}]
 }
 MainTrail {
  emitterUpdateScript [
   &EmitterState(Life Cycle Mode=::NewEnumerator1, Loop Behavior=::NewEnumerator1, Loop Duration=0.5, UseLoopDelay=false)
   &FN_SpawnAnimTrail()
  ]
  particleSpawnScript [
   &FN_InitializeAnimTrailBoneIndecies(First Bone or Socket Index=1, Second Bone or Socket Index=0)
   &FN_InitializeAnimTrail(AnimTrailWidth=0.8, Color=<rgba 0.5312 0.5312 0.5312 1>, Lifetime=0.16)
  ]
  particleUpdateScript [
   &FN_UpdateAnimTrail()
   &ScaleColor(
    Scale Alpha=&FloatFromCurve(FloatCurve=DataInterface() {propertyValues {Curve {keys [<at 0 1>,<at 0.2 1>,<at 0.3 0>]}}}, CurveIndex=$Emitter.Age)
    Scale RGB=&VectorFromFloat(Value=&Multiply_Float(A=1, B=&FNTimeOfDayScalar(Night Scale=0.7, Morning Scale=0.7, Evening Scale=0.7)))
    ScaleA=true
    ScaleRGB=true
   )
   &DynamicMaterialParameters(
    Index 0 Param 1=&Lerp_Float(A=0, B=1, Alpha=$User.VariantSwitch)
    Module.Index0Float0Write=true
    Module.Index0Float1Write=false
    Module.Index0Float2Write=false
    Module.Index0Float3Write=false
   )
  ]
  rendererClasses [&NiagaraRibbonRendererProperties]
  renderers [{rendererClass=&NiagaraRibbonRendererProperties}]
 }
 Outline {
  emitterUpdateScript [
   &EmitterState(Life Cycle Mode=::NewEnumerator1, Loop Behavior=::NewEnumerator1, Loop Duration=0.5, UseLoopDelay=false)
   &FN_SpawnAnimTrail()
  ]
  particleSpawnScript [
   &FN_InitializeAnimTrailBoneIndecies()
   &FN_InitializeAnimTrail(AnimTrailWidth=1.1, Color=<rgba 0.5312 0.5312 0.5312 1>, Lifetime=0.16)
  ]
  particleUpdateScript [
   &FN_UpdateAnimTrail()
   &ScaleColor(
    Scale Alpha=&FloatFromCurve(
     FloatCurve=DataInterface() {propertyValues {Curve {keys [<at 0 1 RCIM_Cubic>,<at 1 0 RCIM_Cubic>]}}}
     Scale Curve=&Multiply_Float(
      A=&FloatFromCurve(FloatCurve=DataInterface() {propertyValues {Curve {keys [<at 0 1>,<at 0.2 1>,<at 0.3 0>]}}}, CurveIndex=$Emitter.Age)
      B=&FNTimeOfDayScalar(Night Scale=0.8, Morning Scale=0.8, Evening Scale=0.8)
     )
    )
    Scale RGB=&MakeVectorFromLinearColorRGB(Color=&LerpLinearColors(
      StartColor=&ColorFromCurve(ColorCurve=DataInterface() {propertyValues {
         AlphaCurve {keys []}
         BlueCurve {keys [{arriveTangent=-17.2603, interpMode=RCIM_Cubic, leaveTangent=-17.2603, tangentMode=RCTM_User, time=0.06, value=8}, <at 0.6278 1 RCIM_Cubic>]}
         GreenCurve {keys [
           {arriveTangent=-14.6689, interpMode=RCIM_Cubic, leaveTangent=-14.6689, tangentMode=RCTM_User, time=0.06, value=5.8271}
           <at 0.6278 0.1416 RCIM_Cubic>
          ]}
         RedCurve {keys [<at 0.06 0 RCIM_Cubic>,<at 0.6278 0 RCIM_Cubic>]}
        }})
      EndColor=&ColorFromCurve(ColorCurve=DataInterface() {propertyValues {
         AlphaCurve {keys []}
         BlueCurve {keys [{arriveTangent=-17.2603, time=0.06, value=0.3124}, <at 0.6278 0.0143>]}
         GreenCurve {keys [{arriveTangent=-14.6689, time=0.06, value=5}, <at 0.6278 0.1823>]}
         RedCurve {keys [<at 0.06 2.692>,<at 0.6278 0.1108>]}
        }})
      LerpFactor=$User.VariantSwitch
     ))
    ScaleA=true
    ScaleRGB=true
   )
  ]
  rendererClasses [&NiagaraRibbonRendererProperties]
  renderers [{rendererClass=&NiagaraRibbonRendererProperties}]
 }
 Particle {
  emitterUpdateScript [
   &EmitterState(Life Cycle Mode=::NewEnumerator1, Loop Behavior=::NewEnumerator1, Loop Duration=0.2, UseLoopDelay=false)
   &SpawnRate(SpawnRate=50, Use Spawn Probability=false)
  ]
  particleSpawnScript [
   &InitializeParticle(
    Color=&LerpLinearColors(
     StartColor=&MakeCustomLinearColorFromBool(Bool=&RandomBool(), True Color=<rgba 0 0.07 0.6 1.5>, False Color=<rgba 0 0.7497 1.5 1>)
     EndColor=&MakeCustomLinearColorFromBool(Bool=&RandomBool(), True Color=<rgba 0.1106 0.1979 0.0158 1.5>, False Color=<rgba 0.5651 1 0.1302 1>)
     LerpFactor=$User.VariantSwitch
    )
    Lifetime Max=0.5
    Lifetime Min=0.2
    Lifetime Mode=::NewEnumerator1
    Sprite Size Mode=::NewEnumerator1
    Uniform Sprite Size Max=4
    Uniform Sprite Size Min=1.5
    UsePositionOffset=false
    Write Material Random=true
   )
   &ShapeLocation(MaskSpawnGroup=false, Override Seed=false, Sphere Radius=10, UseOwnerScale=true)
   &AddVelocity(Velocity=&RandomRangeVector(Minimum=<0 0 80>, Maximum=<0 0 100>))
  ]
  particleUpdateScript [
   &ParticleState()
   &ScaleColor(
    Scale Alpha=&FloatFromCurve(FloatCurve=DataInterface() {propertyValues {Curve {keys [<at 0 1>,<at 0.7 1>,<at 1 0>]}}})
    Scale RGB=<2 2 2>
    ScaleA=true
    ScaleRGB=true
   )
   /SetVariables(Particles.SizeGlow:Vector2f=&Multiply_Vector2D(A=&Vector2DFromFloat(Value=1.8), B=$Particles.SpriteSize)) {isSetParameters=true}
   /SetVariables(Particles.ColorGlow:LinearColor=&MakeLinearColorFromVectorAndFloat(
     "Vector (RGB)"=&MakeVectorFromLinearColorRGB(Color=$Particles.Color)
     "Float (Alpha)"=&FloatFromCurve(FloatCurve=DataInterface() {propertyValues {Curve {keys [<at 0 1>,<at 0.7 1>,<at 1 0>]}}}, Scale Curve=0.5)
    )) {isSetParameters=true}
   &CurlNoiseForce(Bias Noise=false, Noise Frequency=100, Noise Strength=300, Pan Noise=false)
   &Drag(Drag=2, Use Linear Drag=true, Use Rotational Drag=true)
   &SolveForcesAndVelocity(Clamp Velocity=false, Limit Acceleration=false)
  ]
  rendererClasses [&NiagaraSpriteRendererProperties]
  renderers [
   {rendererClass=&NiagaraSpriteRendererProperties}
   {rendererClass=&NiagaraSpriteRendererProperties, rendererIndex=1}
  ]
 }
}
summary {
 systemName=NS_Pickaxe_RadioPaca_TrailOffHand
 emitters [
  {emitterName=Particle, rendererClasses [&NiagaraSpriteRendererProperties]}
  {emitterName=MainTrail, rendererClasses [&NiagaraRibbonRendererProperties]}
  {emitterName=Outline, rendererClasses [&NiagaraRibbonRendererProperties]}
  {emitterName=Centerglow, rendererClasses [&NiagaraRibbonRendererProperties]}
 ]
 userVariables [
  User.RibbonIDOffset {type=NiagaraInt32}
  User.VariantSwitch {type=NiagaraFloat, defaultValue=0, description=""}
 ]
}
