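# Animation Blueprint for the offhand RadioPaca pickaxe that spins its disc and adds physical jiggle to a needle.
# Each update eases a spin step toward a fixed negative value, adds it to an accumulated angle and turns that into a roll rotator, so spin is per frame, not time-scaled.
# The AnimGraph applies it to the disc_spin bone of the input pose, then runs Anim Dynamics with an angular spring and tight limits on the needle bone.
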AnimBlueprint /BRCosmetics/Weapons/FORT_Melee/Pickaxe_RadioPaca/Meshes/RadioPacaOffhand_Axe_AnimBP
target_skeleton=&/RadioPaca_Axe_Skeleton
graphs [
 AnimGraph {nodes [
   AnimGraphNode_SubInput_0 = Input Pose()
   n1 = "Transform (Modify) Bone - Bone: disc_spin"(
    ComponentPose=Local To Component(LocalPose=@AnimGraphNode_SubInput_0)
    Alpha=1
    Translation="0,0,0"
    Rotation=Get Disc Rotator() -> (Disc Rotator:Rotator)
    Scale="1,1,1"
   ) {rotationMode=BMM_Replace, rotationSpace=BCS_BoneSpace, alphaBoolBlend {blendOption=Linear}, boneToModify {boneName=disc_spin}}
   n2 = Anim Dynamics - Bone: needle(ComponentPose=@n1, Alpha=1, ExternalForce="0,0,0") {
    angularDampingOverride=0.9999
    angularSpringConstant=8000
    bAngularSpring=true
    bOverrideAngularDamping=true
    bUseGravityOverride=true
    simulationSpace=BoneRelative
    alphaBoolBlend {blendOption=Linear}
    boundBone {boneName=needle}
    componentAppliedLinearAccClamp=<600 600 600>
    componentLinearAccScale=<0.3 0.3 0.3>
    componentLinearVelScale=<0.3 0.3 0.3>
    gravityOverride {z=-980}
    physicsBodyDefinitions [{
      boundBone {boneName=needle}
      boxExtents=<20 20 20>
      constraintSetup {angularLimitsMax=<0 0 1.5>, angularLimitsMin=<0 0 -2>, angularTarget=<1 0 0>}
      localJointOffset=<2 0 0>
     }]
   }
   Output Pose(Result=Component To Local(ComponentPose=@n2)) {layerGroup=None}
  ]}
 EventGraph {nodes [
   n1 = Event BlueprintUpdateAnimation()
   > n2 = Sequence()
   > Set disc rotation(disc rotation:double=FInterpTo(Current=Get disc rotation() -> (disc rotation:double), Target=-3, DeltaTime=@n1.DeltaTimeX, InterpSpeed=10)) -> (Output_Get:double)
   n3 = Set Rotator(execute=@n2.then_1, Rotator:double=float + float(A=Get Rotator() -> (Rotator:double), B=Get disc rotation())) -> (Output_Get:double)
   > Set Disc Rotator(Disc Rotator:Rotator=MakeRotator(Roll=@n3)) -> (Output_Get:Rotator)
   TryGetPawnOwner()
  ]}
]
variables [
 disc rotation {type=number}
 Disc Rotator {type=Rotator}
 Rotator {type=number}
]
